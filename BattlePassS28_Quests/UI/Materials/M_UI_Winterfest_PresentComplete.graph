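# Translucent UI material for the Winterfest present-complete graphic in the BattlePassS28 quests UI.
# It samples the present-complete texture through an overridable texture parameter, using normalized UVs from GetUserInterfaceUV.
# A vector parameter named colour defaults to bright green; the listed graph shows no links from it or the texture sample to the material outputs.
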Material /BattlePassS28_Quests/UI/Materials/M_UI_Winterfest_PresentComplete
graph {
 properties {
  bUsedWithStaticMesh=false
  blendMode=BLEND_Translucent
  materialDomain=MD_UI
  pixelDepthOffsetMode=PDOM_Legacy
  usedShadingModels=Default Lit
 }
 nodes [
  TextureSampleParameter2D(UVs=MaterialFunctionCall() {materialFunction=&GetUserInterfaceUV}.Normalized UV) {parameterName=Texture, texture=/BattlePassS28_Quests/UI/Textures/T_UI_Winterfest_PresentComplete}
  VectorParameter() {parameterName=colour, defaultValue=<rgba 0.0343 1 0.227 1>}
 ]
}
