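# UI material that draws a Winterfest present tag for the Battle Pass quests screens, with a gradient fill and borders.
# A TagType vector dotted with the distance-field texture's RGB selects the tag shape.
# SmoothStep thresholds on that field, offset cumulatively from the stroke thickness, separate edge stroke, inset border band and tag fill.
# A shared radial gradient blends each region's color pair, while the sharpened field provides the translucent silhouette.
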Material /BattlePassS28_Quests/UI/Materials/M_UI_Winterfest_PresentTag
graph {
 properties {
  bUsedWithStaticMesh=false
  blendMode=BLEND_Translucent
  materialDomain=MD_UI
  pixelDepthOffsetMode=PDOM_Legacy
  usedShadingModels=Default Lit
 }
 nodes [
  n1 = MaterialFunctionCall!(
   CenterPosition=AppendVector(
    A=ScalarParameter() {defaultValue=0.5, group=RadialGradient, parameterName=PositionY}
    B=ScalarParameter() {defaultValue=0.5, group=RadialGradient, parameterName=PositionX}
   )
   Col_01
   Col_02
   Radius Scale=ScalarParameter() {defaultValue=1, group=RadialGradient, parameterName=Scale}
   Density=ScalarParameter() {defaultValue=1, group=RadialGradient, parameterName=Density}
  ) {materialFunction=/Game/UI/Materials/MF_UI_SmoothPoly_RadialGradient}
  n2 = VectorParameter() {parameterName=CenterBorderColor_Outer, defaultValue=<rgba 0.0364 0 0.9844 1>}
  n3 = VectorParameter() {parameterName=CenterBorderColor_Inner, defaultValue=<rgba 0.0498 0.7994 1 1>}
  n4 = VectorParameter() {parameterName=TagColor_Outer, defaultValue=<rgba 0.9479 0.4775 0.1265 1>}
  n5 = VectorParameter() {parameterName=TagColor_Inner, defaultValue=<rgba 0.9479 0.1234 0.1726 1>}
  n6 = ScalarParameter() {defaultValue=0.25, parameterName=StrokeThickness}
  n7 = Add(A=@n6, B=ScalarParameter() {defaultValue=0.25, parameterName=OuterBorderEdge})
  n8 = ScalarParameter() {defaultValue=0.001, parameterName=OuterBorder_SSRange}
  n9 = Add(A=@n7, B=Multiply(A=@n8) {constB=-1})
  n10 = Add(A=@n7, B=@n8)
  n11 = DotProduct(
   A=TextureSample(UVs=TextureCoordinate()) {samplerType=SAMPLERTYPE_LinearColor, texture=/BattlePassS28_Quests/UI/Textures/T_UI_DF_Winterfest_PresentTag}.RGB
   B=VectorParameter() {group=MaskSwitches, parameterName=TagType, defaultValue=<rgba 1 0 0 1>}.RGB
  )
  n12 = OneMinus(None=Saturate(None=MaterialFunctionCall(Alpha=@n11, Min=@n9, Max=@n10) {materialFunction=&SmoothStep}.Result))
  n13 = Add(A=@n7, B=ScalarParameter() {defaultValue=0.25, parameterName=InnerBorderEdge})
  n14 = ScalarParameter() {defaultValue=0.001, parameterName=InnerBorder_SSRange}
  n15 = Multiply(A=@n14) {constB=-1}
  n16 = Saturate(None=MaterialFunctionCall(Alpha=@n11, Min=Add(A=@n13, B=@n15), Max=Add(A=@n13, B=@n14)) {materialFunction=&SmoothStep}.Result)
  n17 = ScalarParameter() {defaultValue=0.001, parameterName=StrokeThickness_SSRange}
  n18 = Add(A=@n6, B=Multiply(A=@n17) {constB=-1})
  n19 = Add(A=@n6, B=@n17)
  LinearInterpolate(
   A=LinearInterpolate(
    A=VectorParameter() {parameterName=EdgeBorderColor_Outer, defaultValue=<rgba 0.9844 0 0.2661 1>}.RGB
    B=VectorParameter() {parameterName=EdgeBorderColor_Inner, defaultValue=<rgba 0.2134 0 1 1>}.RGB
    Alpha=@n1."Gradient (0-1)"
   )
   B=LinearInterpolate(
    A=LinearInterpolate(A=@n2.RGB, B=@n3.RGB, Alpha=@n1."Gradient (0-1)")
    B=LinearInterpolate(A=@n4.RGB, B=@n5.RGB, Alpha=@n1."Gradient (0-1)")
    Alpha=LinearInterpolate(A=@n12, Alpha=@n16)
   )
   Alpha=Saturate(None=MaterialFunctionCall(Alpha=@n11, Min=@n18, Max=@n19) {materialFunction=&SmoothStep}.Result)
  )
  n20 = ScalarParameter() {defaultValue=32, parameterName=OuterShapeSharpness}
  Reroute(None=Saturate(None=Multiply(A=@n11, B=@n20) {constB=48}))
 ]
}
