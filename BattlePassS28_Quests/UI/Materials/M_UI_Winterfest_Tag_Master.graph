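# Translucent UI master material that procedurally draws a Winterfest tag shape with striped fill, border and round hole.
# Five rotated, offset edge gradients (top, bottom, left, two angled right) are clamped and combined via scalar DotProducts into the silhouette.
# An inset copy of those edges separates BorderColor from the fill, where a sine of rotated UVs blends TagColor and LinesColor.
# A RadialGradientExponential circle is cut from the outer silhouette mask.
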Material /BattlePassS28_Quests/UI/Materials/M_UI_Winterfest_Tag_Master
graph {
 properties {
  bUsedWithStaticMesh=false
  blendMode=BLEND_Translucent
  materialDomain=MD_UI
  pixelDepthOffsetMode=PDOM_Legacy
  usedShadingModels=Default Lit
 }
 nodes [
  n1 = Multiply(
   A=MaterialFunctionCall!(
    Rotation_Fraction=ScalarParameter() {defaultValue=-0.26, group=Lines, parameterName=Lines_Angle}
    TexCoord=TextureCoordinate() {coordinateIndex=4}
   ) {materialFunction=/FortUILibrary/Foundation/Art_ToolKit/AngledBoxes/MF_UI_Rotator_Fractions}.Result
   B=ScalarParameter() {defaultValue=5, group=Lines, parameterName=LinesAmount}
  )
  n2 = Multiply(A=ComponentMask(None=Sine(None=@n1)) {g=true}, B=ScalarParameter() {defaultValue=16, group=Lines, parameterName=LinesSoftness})
  n3 = MaterialFunctionCall(
   Rotation_Fraction=ScalarParameter() {defaultValue=-0.24, group=BottomAndTop, parameterName=Angle_Top}
   TexCoord=TextureCoordinate() {coordinateIndex=4}
  ) {materialFunction=&MF_UI_Rotator_Fractions}
  n4 = ScalarParameter() {defaultValue=-0.25, group=BottomAndTop, parameterName=Offset_Top}
  n5 = Add(
   A=@n3.Result
   B=Add(A=@n4, B=ScalarParameter() {defaultValue=-0.01, group=BottomAndTop, parameterName=Inner_OffsetAmount_Y}) {constB=-0.0075}
  )
  n6 = ScalarParameter() {defaultValue=48, parameterName=Softness_Top}
  n7 = Multiply(A=Saturate(None=ComponentMask(None=@n5) {r=true}), B=@n6)
  n8 = MaterialFunctionCall(
   Rotation_Fraction=ScalarParameter() {defaultValue=-0.24, group=BottomAndTop, parameterName=Angle_Bottom}
   TexCoord=TextureCoordinate() {coordinateIndex=4}
  ) {materialFunction=&MF_UI_Rotator_Fractions}
  n9 = ScalarParameter() {defaultValue=0.25, group=BottomAndTop, parameterName=Offset_Bottom}
  n10 = Multiply(A=ScalarParameter() {defaultValue=-0.01, group=BottomAndTop, parameterName=Inner_OffsetAmount_Y}) {constB=-1}
  n11 = ComponentMask(None=Add(A=@n8.Result, B=Add(A=@n9, B=@n10) {constB=-0.0075})) {r=true}
  n12 = ScalarParameter() {defaultValue=48, parameterName=Softness_Bottom}
  n13 = Multiply(A=Saturate(None=OneMinus(None=@n11)), B=@n12)
  n14 = MaterialFunctionCall(Rotation_Fraction=ScalarParameter() {defaultValue=-0.025, group=Left, parameterName=Angle_Left}, TexCoord=TextureCoordinate()) {materialFunction=&MF_UI_Rotator_Fractions}
  n15 = ScalarParameter() {defaultValue=-0.075, group=Left, parameterName=Offset_Left}
  n16 = Add(A=@n14.Result, B=Add(A=@n15, B=ScalarParameter() {defaultValue=-0.01, parameterName=Inner_OffsetAmount}) {constB=-0.0075})
  n17 = ScalarParameter() {defaultValue=48, group=Left, parameterName=Softness_Left}
  n18 = Multiply(A=Saturate(None=ComponentMask(None=@n16) {r=true}), B=@n17)
  n19 = MaterialFunctionCall(Rotation_Fraction=ScalarParameter() {defaultValue=0.075, group=Right, parameterName=Angle_Right_Top}, TexCoord=TextureCoordinate()) {materialFunction=&MF_UI_Rotator_Fractions}
  n20 = ScalarParameter() {group=Right, parameterName=Offset_Right_Top}
  n21 = Add(A=@n20, B=Multiply(A=ScalarParameter() {defaultValue=-0.01, group=Right, parameterName=Inner_OffsetAmount_RightSide}) {constB=-1}) {constB=-0.0075}
  n22 = OneMinus(None=ComponentMask(None=Add(A=@n19.Result, B=@n21)) {r=true})
  n23 = ScalarParameter() {defaultValue=48, parameterName=Softness_Right_Top}
  n24 = Saturate(None=Multiply(A=Saturate(None=@n22), B=@n23))
  n25 = MaterialFunctionCall(Rotation_Fraction=ScalarParameter() {defaultValue=-0.075, group=Right, parameterName=Angle_Right_Bottom}, TexCoord=TextureCoordinate()) {materialFunction=&MF_UI_Rotator_Fractions}
  n26 = ScalarParameter() {group=Right, parameterName=Offset_Right_Bottom}
  n27 = Add(A=@n26, B=Multiply(A=ScalarParameter() {defaultValue=-0.01, group=Right, parameterName=Inner_OffsetAmount_RightSide}) {constB=-1}) {constB=-0.0075}
  n28 = OneMinus(None=ComponentMask(None=Add(A=@n25.Result, B=@n27)) {r=true})
  n29 = ScalarParameter() {defaultValue=48, parameterName=Softness_Right_Bottom}
  n30 = Saturate(None=Multiply(A=Saturate(None=@n28), B=@n29))
  n31 = DotProduct(A=DotProduct(A=Saturate(None=@n7), B=Saturate(None=@n13)), B=DotProduct(A=Saturate(None=@n18), B=DotProduct(A=@n24, B=@n30)))
  LinearInterpolate(
   A=LinearInterpolate(
    A=VectorParameter() {group=Color, parameterName=TagColor, defaultValue=<rgba 1 0 0 1>}.RGB
    B=VectorParameter() {group=Lines, parameterName=LinesColor, defaultValue=<rgba 0 0 1 1>}.RGB
    Alpha=Saturate(None=@n2)
   )
   B=VectorParameter() {group=Color, parameterName=BorderColor, defaultValue=<rgba 0 1 0 1>}.RGB
   Alpha=OneMinus(None=Saturate(None=@n31))
  )
  n32 = Add(A=@n3.Result, B=@n4)
  n33 = Multiply(A=Saturate(None=ComponentMask(None=@n32) {r=true}), B=@n6)
  n34 = ComponentMask(None=Add(A=@n8.Result, B=@n9)) {r=true}
  n35 = Multiply(A=Saturate(None=OneMinus(None=@n34)), B=@n12)
  n36 = Add(A=@n14.Result, B=@n15)
  n37 = Multiply(A=Saturate(None=ComponentMask(None=@n36) {r=true}), B=@n17)
  n38 = OneMinus(None=ComponentMask(None=Add(A=@n19.Result, B=@n20)) {r=true})
  n39 = Saturate(None=Multiply(A=Saturate(None=@n38), B=@n23))
  n40 = OneMinus(None=ComponentMask(None=Add(A=@n25.Result, B=@n26)) {r=true})
  n41 = Saturate(None=Multiply(A=Saturate(None=@n40), B=@n29))
  n42 = DotProduct(A=DotProduct(A=Saturate(None=@n33), B=Saturate(None=@n35)), B=DotProduct(A=Saturate(None=@n37), B=DotProduct(A=@n39, B=@n41)))
  n43 = TextureCoordinate()
  n44 = AppendVector(A=ScalarParameter() {defaultValue=1, parameterName=Circle_Tile_X}, B=ScalarParameter() {defaultValue=1, parameterName=Circle_Tile_Y})
  n45 = ScalarParameter() {defaultValue=0.75, parameterName=Circle_Center_U}
  n46 = ScalarParameter() {defaultValue=0.5, parameterName=Circle_Center_V}
  n47 = Saturate(None=MaterialFunctionCall(
    UVs=Multiply(A=@n43, B=@n44)
    CenterPosition=AppendVector(A=@n45, B=@n46)
    Radius=ScalarParameter() {defaultValue=0.1, parameterName=Circle_Radius}
    Density=ScalarParameter() {defaultValue=16, parameterName=Circle_Softness}
   ) {materialFunction=&RadialGradientExponential}.RadialGradientExponential)
  Saturate(None=DotProduct(A=Saturate(None=@n42), B=OneMinus(None=@n47)))
 ]
}
