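# One-shot stereo MetaSound for the Season 28 Winterfest present-opening UI sound, layering a present crash with a Winterfest jingle.
# Chained trigger delays start the crash shortly after play and the jingle just behind it.
# An attack-decay envelope shapes only the crash's gain before a two-input mixer sums both layers.
# The crash player alone signals finish, and concurrency allows one instance, stopping the oldest.
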MetaSoundSource /BattlePassS28_Quests/Audio/MSS_S28_UI_Winterfest_Present_Open
inputs [
 UE.Source.OnPlay {type=Trigger}
 Wave Asset {type=WaveAsset}
 Wave Asset 1 {type=WaveAsset}
]
interfaces [UE.OutputFormat.Stereo:1.0, UE.Source.OneShot:1.0, UE.Source:1.1]
outputs [
 "UE.OutputFormat.Stereo.Audio:0" {type=Audio}
 "UE.OutputFormat.Stereo.Audio:1" {type=Audio}
 UE.Source.OneShot.OnFinished {type=Trigger}
]
pages [Default {
  input_defaults {
   UE.Source.OnPlay=null
   Wave Asset=/Game/Athena/Sounds/Events/Winterfest/Winterfest_Jingle_01
   Wave Asset 1=/Game/Sounds/Fort_Gadgets/Present/Present_PlaceCrash_01
  }
  nodes [
   TriggerDelay_0 = UE.Trigger Delay(In=^UE.Source.OnPlay) {also {defaults {Delay Time=0.2}}}
   WavePlayer_0 = UE.Wave Player:Stereo(Wave Asset=^Wave Asset 1, Play=@TriggerDelay_0.out)
   ADEnvelopeADEnvelope_0 = AD Envelope.AD Envelope:Float(Trigger=@WavePlayer_0.On Play) {also {defaults {Attack Time=0.15, Decay Time=5}}}
   TriggerDelay_1 = UE.Trigger Delay(In=@TriggerDelay_0.out) {also {defaults {Delay Time=0.1}}}
   WavePlayer_1 = UE.Wave Player:Stereo(Wave Asset=^Wave Asset, Play=@TriggerDelay_1.out)
   AudioMixerAudioMixerStereo2_0 = "AudioMixer.Audio Mixer (Stereo, 2)"(
    Gain 1=@ADEnvelopeADEnvelope_0.Out Envelope
    In 1 L=@WavePlayer_0.Out Left
    In 1 R=@WavePlayer_0.Out Right
    In 0 L=@WavePlayer_1.Out Left
    In 0 R=@WavePlayer_1.Out Right
   )
  ]
  links [
   {dst="@UE.OutputFormat.Stereo.Audio:0", src=AudioMixerAudioMixerStereo2_0.Out L}
   {dst="@UE.OutputFormat.Stereo.Audio:1", src=AudioMixerAudioMixerStereo2_0.Out R}
   {dst="@UE.Source.OneShot.OnFinished", src=WavePlayer_0.On Finished}
  ]
 }]
properties {
 assetImportData=&self:AssetImportData
 bOverrideConcurrency=true
 numChannels=2
 outputFormat=Stereo
 sampleRate=48000
 soundClassObject=&Fort_UI
 concurrencyOverrides {maxCount=1, resolutionRule=StopOldest, retriggerTime=0.11}
 modulationSettings {volumeModulationDestination {bEnableModulation=true, value=-60, modulators [null]}}
}
