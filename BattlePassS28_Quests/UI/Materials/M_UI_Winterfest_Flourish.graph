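# UI-domain translucent material for the Winterfest flourish graphic in the Battle Pass S28 quests UI.
# It samples a flourish texture parameter using normalized UVs from GetUserInterfaceUV.
# A separate colour vector parameter defaults to a pale lavender-white with zero alpha; its link to the texture or material outputs is not shown.
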Material /BattlePassS28_Quests/UI/Materials/M_UI_Winterfest_Flourish
graph {
 properties {
  bUsedWithStaticMesh=false
  blendMode=BLEND_Translucent
  materialDomain=MD_UI
  pixelDepthOffsetMode=PDOM_Legacy
  usedShadingModels=Default Lit
 }
 nodes [
  TextureSampleParameter2D(UVs=MaterialFunctionCall() {materialFunction=&GetUserInterfaceUV}.Normalized UV) {parameterName=Texture, texture=/BattlePassS28_Quests/UI/Textures/T_UI_Winterfest_Flourish}
  VectorParameter() {parameterName=colour, defaultValue=<rgba 0.9484 0.9167 1 0>}
 ]
}
