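# M_UI_Winterfest_ShipItParcel is a translucent UI material for the Winterfest "Ship It" parcel image in the BattlePassS28 quests content.
# It samples the T_UI_Winterfest_ShipItParcel texture through a parameter named Texture, so material instances can swap the image.
# The sample uses the Normalized UV output of GetUserInterfaceUV, with no other processing in the graph.
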Material /BattlePassS28_Quests/UI/Materials/M_UI_Winterfest_ShipItParcel
graph {
 properties {
  bUsedWithStaticMesh=false
  blendMode=BLEND_Translucent
  materialDomain=MD_UI
  pixelDepthOffsetMode=PDOM_Legacy
  usedShadingModels=Default Lit
 }
 nodes [TextureSampleParameter2D(UVs=MaterialFunctionCall() {materialFunction=&GetUserInterfaceUV}.Normalized UV) {parameterName=Texture, texture=/BattlePassS28_Quests/UI/Textures/T_UI_Winterfest_ShipItParcel}]
}
