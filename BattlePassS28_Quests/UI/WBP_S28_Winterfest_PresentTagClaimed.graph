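# Widget Blueprint for the "claimed" tag shown on a present in the Season 28 Winterfest quests UI.
# A slightly tilted, sheared overlay layers a horizontally stretchable Box-drawn tag material image behind a padded text block that reads "CLAIMED!" by default.
# A SetText function replaces that label, while the Construct, PreConstruct and Tick events are empty.
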WidgetBlueprint /BattlePassS28_Quests/UI/WBP_S28_Winterfest_PresentTagClaimed
graphs [
 EventGraph {nodes [
   Event PreConstruct()
   Event Construct()
   Event Tick()
  ]}
 SetText {nodes [
   n1 = SetText!() -> (then, Text:Text) {is=title K2Node_FunctionEntry}
   > "SetText (Text)"(self=Get CommonTextBlock_56() -> (CommonTextBlock_56:Common Text Object Reference), InText=@n1)
  ]}
]
variables []
widgets {
 info {inheritedWidgetCount=0, parentClass=&UserWidget, rootWidgetClass=&Overlay, widgetCount=5}
 widgets [
  Overlay_34 {
   widgetClassPath=&Overlay
   widget=Overlay_34
   properties {renderTransform {angle=-3, scale=<1 1>, shear=<-9 -1>, translation=<0 0>}}
  }
  Tag {
   widgetClassPath=&Image
   bIsVariable=true
   parent=Overlay_34
   widget=Tag
   properties {brush {
     bIsDynamicallyLoaded=false
     drawAs=Box
     imageType=NoImage
     mirroring=NoMirror
     resourceName=None
     resourceObject=/BattlePassS28_Quests/UI/Materials/MI_UI_Winterfest_PresentTag
     tiling=NoTile
     imageSize=<70 70>
     margin=<ltrb 0.5 0 0.5 0>
     outlineSettings {
      bUseBrushTransparency=false
      roundingType=HalfHeightRadius
      width=0
      color {colorUseRule=UseColor_Specified, specifiedColor=<rgba 0 0 0 0>}
      cornerRadii=<0 0 0 0>
     }
     tintColor {colorUseRule=UseColor_Specified, specifiedColor=<rgba 1 1 1 1>}
     uVRegion {bIsValid=false, max=<0 0>, min=<0 0>}
    }}
   slot {class=OverlaySlot, properties {horizontalAlignment=HAlign_Fill, verticalAlignment=VAlign_Center}}
  }
  ContentHB {
   widgetClassPath=&HorizontalBox
   parent=Overlay_34
   widget=ContentHB
   slot {class=OverlaySlot, properties {horizontalAlignment=HAlign_Fill, verticalAlignment=VAlign_Center, padding=<ltrb 40 0 40 0>}}
  }
  VerticalBox_88 {
   widgetClassPath=&VerticalBox
   parent=ContentHB
   widget=VerticalBox_88
   slot {class=HorizontalBoxSlot, properties {verticalAlignment=VAlign_Center}}
  }
  CommonTextBlock_56 {
   widgetClassPath=&CommonTextBlock
   bIsVariable=true
   parent=VerticalBox_88
   widget=CommonTextBlock_56
   properties {
    text="CLAIMED!"
    colorAndOpacity {colorUseRule=UseColor_Specified, specifiedColor=<rgba 0.4678 0.0137 0 1>}
    font {
     bForceMonospaced=false
     bMaterialIsStencil=false
     fontMaterial=null
     fontObject=&HeadingNow
     letterSpacing=1
     monospacedWidth=1
     size=21
     skewAmount=0
     typefaceFontName=75 Medium
     outlineSettings {
      bApplyOutlineToDropShadows=false
      bMiteredCorners=false
      bSeparateFillAlpha=false
      outlineMaterial=null
      outlineSize=0
      outlineColor=<rgba 0 0 0 1>
     }
    }
    shadowOffset=<0 0>
   }
   slot {class=VerticalBoxSlot, properties {verticalAlignment=VAlign_Center, size {sizeRule=Fill, value=1}}}
  }
 ]
}
