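# Gift card widget on the Season 28 Winterfest goals page, showing Ship It! Express gift status and opening available gifts.
# Populate events set description text from DisplayData and switch ContentSwitcher between countdown, ready, inaccessible, claimed and completed views.
# Clicking sends the CompleteStatTag stat event, then in frontend disables the button and plays sound with a one-second control bus mix; otherwise it shows the claimed view.
# Animations handle hover, selection and floating gifts.
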WidgetBlueprint /BattlePassS28_Quests/UI/WBP_S28_WinterfestGoalsPageGiftCard
animations [
 Hovered {binding_count=0, length_seconds=0.15, object=&self:Hovered, sequence {bindings []}}
 Selected {
  binding_count=3
  length_seconds=0.15
  object=&self:Selected
  sequence {
   bindings [
    Image_Disabled {id=#1, possessed_class=Image, tracks []}
    Image_Fade {
     id=#2
     possessed_class=Image
     tracks [{
       type=MovieSceneFloatTrack
       display_name=Render Opacity
       sections [{type=MovieSceneFloatSection, channels [None {type=MovieSceneScriptingFloatChannel, keys [[0,1],[1,0]]}]}]
      }]
    }
    Text_QuestID {
     id=#3
     possessed_class=CommonTextBlock
     tracks [{
       type=MovieSceneColorTrack
       display_name=Color and Opacity
       sections [{
         type=MovieSceneColorSection
         channels [
          Color.R {type=MovieSceneScriptingFloatChannel, keys [[0,1],[3,0]]}
          Color.G {type=MovieSceneScriptingFloatChannel, keys [[0,1],[3,0]]}
          Color.B {type=MovieSceneScriptingFloatChannel, keys [[0,1],[3,0]]}
          Color.A {type=MovieSceneScriptingFloatChannel, const [0.5,1]}
         ]
        }]
      }]
    }
   ]
   tracks [{
     type=MovieSceneEventTrack
     sections [{
       type=MovieSceneEventTriggerSection
       channels [None {type=MovieSceneScriptingEventChannel, const ["<Struct 'MovieSceneEvent' (0x4666ffbe060) {}>", 1]}]
      }]
    }]
  }
 }
 Idle {
  binding_count=2
  length_seconds=1.5012
  object=&self:Idle
  sequence {bindings [
    ShipItParcel {
     id=#4
     possessed_class=Image
     tracks [{
       type=MovieScene2DTransformTrack
       sections [{
         type=MovieScene2DTransformSection
         end_frame=30
         start_frame=0
         channels [
          Translation.X {type=MovieSceneScriptingFloatChannel, const [0,1]}
          Translation.Y {type=MovieSceneScriptingFloatChannel, keys [[0,0],[15.059,-8],[30.0233,0]]}
          Angle {type=MovieSceneScriptingFloatChannel, const [0,1]}
          Scale.X {type=MovieSceneScriptingFloatChannel, const [1,1]}
          Scale.Y {type=MovieSceneScriptingFloatChannel, const [1,1]}
          Shear.X {type=MovieSceneScriptingFloatChannel, const [0,1]}
          Shear.Y {type=MovieSceneScriptingFloatChannel, const [0,1]}
         ]
        }]
      }]
    }
    ShipItParcel_Outline {
     id=#5
     possessed_class=Image
     tracks [{
       type=MovieScene2DTransformTrack
       sections [{
         type=MovieScene2DTransformSection
         end_frame=30
         start_frame=0
         channels [
          Translation.X {type=MovieSceneScriptingFloatChannel, const [0,1]}
          Translation.Y {type=MovieSceneScriptingFloatChannel, keys [[0,0],[15.059,-8],[30.0233,0]]}
          Angle {type=MovieSceneScriptingFloatChannel, const [0,1]}
          Scale.X {type=MovieSceneScriptingFloatChannel, const [1,1]}
          Scale.Y {type=MovieSceneScriptingFloatChannel, const [1,1]}
          Shear.X {type=MovieSceneScriptingFloatChannel, const [0,1]}
          Shear.Y {type=MovieSceneScriptingFloatChannel, const [0,1]}
         ]
        }]
      }]
    }
   ]}
 }
 GiftCompleted_Float {
  binding_count=5
  length_seconds=1.9986
  object=&self:GiftCompleted_Float
  sequence {bindings [
    GiftCompleted {
     id=#6
     possessed_class=Image
     tracks [{
       type=MovieScene2DTransformTrack
       sections [{
         type=MovieScene2DTransformSection
         end_frame=39
         start_frame=0
         channels [
          Translation.X {type=MovieSceneScriptingFloatChannel, const [0,1]}
          Translation.Y {type=MovieSceneScriptingFloatChannel, keys [[0,40],[19.9123,30],[39.9727,40]]}
          Angle {type=MovieSceneScriptingFloatChannel, const [0,1]}
          Scale.X {type=MovieSceneScriptingFloatChannel, const [1,1]}
          Scale.Y {type=MovieSceneScriptingFloatChannel, const [1,1]}
          Shear.X {type=MovieSceneScriptingFloatChannel, const [0,1]}
          Shear.Y {type=MovieSceneScriptingFloatChannel, const [0,1]}
         ]
        }]
      }]
    }
    GiftPreview_01 {
     id=#7
     possessed_class=Image
     tracks [{
       type=MovieScene2DTransformTrack
       sections [{
         type=MovieScene2DTransformSection
         end_frame=39
         start_frame=0
         channels [
          Translation.X {type=MovieSceneScriptingFloatChannel, const [0,2]}
          Translation.Y {type=MovieSceneScriptingFloatChannel, keys [[0,0],[15.6253,-6],[39.9727,0]]}
          Angle {type=MovieSceneScriptingFloatChannel, const [0,1]}
          Scale.X {type=MovieSceneScriptingFloatChannel, const [1,1]}
          Scale.Y {type=MovieSceneScriptingFloatChannel, const [1,1]}
          Shear.X {type=MovieSceneScriptingFloatChannel, const [0,1]}
          Shear.Y {type=MovieSceneScriptingFloatChannel, const [0,1]}
         ]
        }]
      }]
    }
    GiftPreview_02 {
     id=#8
     possessed_class=Image
     tracks [{
       type=MovieScene2DTransformTrack
       sections [{
         type=MovieScene2DTransformSection
         end_frame=39
         start_frame=0
         channels [
          Translation.X {type=MovieSceneScriptingFloatChannel, const [0,1]}
          Translation.Y {type=MovieSceneScriptingFloatChannel, keys [[0,74],[23.6333,80],[39.9727,74]]}
          Angle {type=MovieSceneScriptingFloatChannel, const [0,1]}
          Scale.X {type=MovieSceneScriptingFloatChannel, const [1,1]}
          Scale.Y {type=MovieSceneScriptingFloatChannel, const [1,1]}
          Shear.X {type=MovieSceneScriptingFloatChannel, const [0,1]}
          Shear.Y {type=MovieSceneScriptingFloatChannel, const [0,1]}
         ]
        }]
      }]
    }
    GiftPreview_04 {
     id=#9
     possessed_class=Image
     tracks [{
       type=MovieScene2DTransformTrack
       sections [{
         type=MovieScene2DTransformSection
         end_frame=39
         start_frame=0
         channels [
          Translation.X {type=MovieSceneScriptingFloatChannel, const [0,1]}
          Translation.Y {type=MovieSceneScriptingFloatChannel, keys [[0,74],[25.7363,80],[39.9727,74]]}
          Angle {type=MovieSceneScriptingFloatChannel, const [0,1]}
          Scale.X {type=MovieSceneScriptingFloatChannel, const [1,1]}
          Scale.Y {type=MovieSceneScriptingFloatChannel, const [1,1]}
          Shear.X {type=MovieSceneScriptingFloatChannel, const [0,1]}
          Shear.Y {type=MovieSceneScriptingFloatChannel, const [0,1]}
         ]
        }]
      }]
    }
    GiftPreview_03 {
     id=#10
     possessed_class=Image
     tracks [{
       type=MovieScene2DTransformTrack
       sections [{
         type=MovieScene2DTransformSection
         end_frame=39
         start_frame=0
         channels [
          Translation.X {type=MovieSceneScriptingFloatChannel, const [0,1]}
          Translation.Y {type=MovieSceneScriptingFloatChannel, keys [[0,0],[19.1037,-6],[39.9727,0]]}
          Angle {type=MovieSceneScriptingFloatChannel, const [0,1]}
          Scale.X {type=MovieSceneScriptingFloatChannel, const [1,1]}
          Scale.Y {type=MovieSceneScriptingFloatChannel, const [1,1]}
          Shear.X {type=MovieSceneScriptingFloatChannel, const [0,1]}
          Shear.Y {type=MovieSceneScriptingFloatChannel, const [0,1]}
         ]
        }]
      }]
    }
   ]}
 }
]
graphs [
 EventGraph {nodes [
   "On Clicked (WBP_UIKit_Button_Regular)"!() -> (OutputDelegate, then, Button:Common Button Base Object Reference) {is=title K2Node_ComponentBoundEvent}
   > n1 = Cast To FortPlayerController(Object=Get Owning Player!(self:Widget Object Reference) -> (ReturnValue:Player Controller Object Reference) {is=title K2Node_CallFunction})
   n2 = Get DisplayData() -> (DisplayData:Challenge Bundle Gift Card Display Data Structure)
   > SendClientStatEventWithTags!(
    execute:Exec
    QuestManager:Fort Quest Manager Object Reference=GetQuestManager!(self:Fort Player Controller Object Reference=@n1, SubGame:ESubGame Enum=Invalid) -> (ReturnValue:Fort Quest Manager Object Reference)
    TargetTags:Gameplay Tag Container Structure
    SourceTags:Gameplay Tag Container Structure=MakeGameplayTagContainerFromTag!(SingleTag:Gameplay Tag Structure=Break Challenge Bundle Gift Card Display Data!(ChallengeBundleGiftCardDisplayData:"Challenge Bundle Gift Card Display Data Structure (by ref)"=@n2) -> (ReturnToTheLobbyText:Text, FlushToTheBattleText:Text, CompleteStatTag:Gameplay Tag Structure) {is=title K2Node_BreakStruct}.CompleteStatTag) -> (ReturnValue:Gameplay Tag Container Structure)
    ContextTags:Gameplay Tag Container Structure
    Count:int=1
   ) -> (then:Exec, OnSuccess:Exec, OnIgnored:Exec, OnFailure:Exec) {is=title K2Node_AsyncAction}
   > n3 = Cast To FortGameStateFrontEnd!(execute, Object=GetGameState!() -> (ReturnValue:Game State Base Object Reference)) -> (then, CastFailed, AsFort Game State Front End:Fort Game State Front End Object Reference) {is=title K2Node_DynamicCast}
   > SetIsInteractionEnabled!(
    execute
    self:Common Button Base Object Reference=Get WBP_UIKit_Button_Regular() -> (WBP_UIKit_Button_Regular:WBP UIKit Button Regular Object Reference)
    bInIsInteractionEnabled:bool=false
   ) -> (then)
   > SetText(
    self=Get RichText_Description() -> (RichText_Description:Common Rich Text Block Object Reference)
    InText=MakeLiteralText(Value="LOCTEXT(\"Open...\")")
   )
   > PlaySound2D(Sound=/BattlePassS28_Quests/Audio/MSS_S28_UI_Winterfest_Present_Open)
   > n4 = Activate Control Bus Mix!(execute, Mix:Sound Control Bus Mix Object Reference=/BattlePassS28_Quests/Audio/CBM_S28_UI_Winterfest) -> (then)
   n5 = DeactivateControlBusMix() {is=title K2Node_CustomEvent}
   Set Timer by Event(execute=@n4.then, Delegate=@n5, Time=1, bLooping=false)
   Populate Was Demand During the Battle(execute=@n3.CastFailed)
   n6 = Event BP_OnSelectionChanged!() -> (OutputDelegate, then, bSelected:bool) {is=title K2Node_Event}
   > n7 = Branch(Condition=@n6.bSelected)
   > Play Animation!(
    execute
    self:User Widget Object Reference
    InAnimation:Widget Animation Object Reference=Get Selected() -> (Selected:Widget Animation Object Reference)
    StartAtTime:float=0
    NumLoopsToPlay:int=1
    PlayMode:EUMGSequencePlayMode Enum
    PlaybackSpeed:float=1
    bRestoreState:bool=false
   ) -> (then, ReturnValue:UMGSequence Player Object Reference)
   Play Animation Reverse!(
    execute=@n7.else
    self:User Widget Object Reference
    InAnimation:Widget Animation Object Reference=Get Selected()
    PlaybackSpeed:float=1
    bRestoreState:bool=false
   ) -> (then, ReturnValue:UMGSequence Player Object Reference)
   n8 = Event BP_SetDisabled!() -> (OutputDelegate, then, bDisabled:bool) {is=title K2Node_Event}
   > Play Animation(InAnimation=Get Selected(), StartAtTime=1, NumLoopsToPlay=1, PlayMode=Reverse, PlaybackSpeed=1, bRestoreState=false)
   > Set Visibility!(
    execute
    self:Widget Object Reference=Get Image_Disabled() -> (Image_Disabled:Image Object Reference)
    InVisibility:ESlateVisibility Enum=Select(Option 0=Collapsed, Option 1=HitTestInvisible, Index=@n8.bDisabled)
   ) -> (then) {is=title K2Node_CallFunction}
   Event PopulateAllQuestsAreClaimedBundleIsCompleted() {is=title K2Node_Event}
   > SetText(
    self=Get RichText_Description()
    InText=Break Challenge Bundle Gift Card Display Data!(ChallengeBundleGiftCardDisplayData=Get DisplayData()) -> (AllQuestsAreClaimedText:Text, ReturnToTheLobbyText, FlushToTheBattleText)
   )
   > SetActiveWidget(
    self=Get ContentSwitcher() -> (ContentSwitcher:Common Visibility Switcher Object Reference)
    Widget=Get GiftCompleted_Container() -> (GiftCompleted_Container:Overlay Object Reference)
   )
   > Stop All Animations(self:User Widget Object Reference)
   > Play Animation(
    InAnimation=Get GiftCompleted_Float() -> (GiftCompleted_Float:Widget Animation Object Reference)
    StartAtTime=0
    NumLoopsToPlay=0
    PlaybackSpeed=1
    bRestoreState=false
   )
   Event OnMouseEnter()
   > Branch(Condition=Get Is Selectable!(self:Object Reference) -> (ReturnValue:bool) {is=title K2Node_CallFunction})
   n9 = Get Hovered() -> (Hovered:Widget Animation Object Reference)
   > Play Animation Forward!(
    execute
    self:User Widget Object Reference
    InAnimation:Widget Animation Object Reference=@n9
    PlaybackSpeed:float=1
    bRestoreState:bool=false
   ) -> (then, ReturnValue:UMGSequence Player Object Reference)
   Event OnMouseLeave!() -> (OutputDelegate, then, MouseEvent:"Pointer Event Structure (by ref)") {is=title K2Node_Event}
   > Branch(Condition=Get Is Selectable())
   > Play Animation Reverse(InAnimation=@n9, PlaybackSpeed=1, bRestoreState=false)
   n10 = Event PopulateHasAvailableQuestsForCompletion!() -> (OutputDelegate, then, AvailableQuestsAmount:int=0) {is=title K2Node_Event}
   > SetIsInteractionEnabled(self=Get WBP_UIKit_Button_Regular(), bInIsInteractionEnabled=true)
   > SetActiveWidget(self=Get ContentSwitcher(), Widget=Get GiftReady_Container() -> (GiftReady_Container:Vertical Box Object Reference))
   > n11 = Branch(Condition="integer > integer"(A=@n10.AvailableQuestsAmount, B=1))
   > Set Visibility(self=Get PendingAmountBlock() -> (PendingAmountBlock:Overlay Object Reference), InVisibility=SelfHitTestInvisible)
   > Set Text(
    Text=Format Text!(Format="LOCTEXT(\"x{0}\")", 0:int=@n10.AvailableQuestsAmount) -> (Result)
    self=Get Text_PendingGiftsAmount() -> (Text_PendingGiftsAmount:Common Text Object Reference)
   )
   > SetText(
    self=Get RichText_Description()
    InText=Break Challenge Bundle Gift Card Display Data!(ChallengeBundleGiftCardDisplayData=Get DisplayData()) -> (AllQuestsAreClaimedText, ThereAreMultipleAvailableQuestsText:Text, ReturnToTheLobbyText, FlushToTheBattleText).ThereAreMultipleAvailableQuestsText
   )
   Set Visibility(execute=@n11.else, self=Get PendingAmountBlock(), InVisibility=Collapsed)
   > SetText(
    self=Get RichText_Description()
    InText=Break Challenge Bundle Gift Card Display Data!(ChallengeBundleGiftCardDisplayData=Get DisplayData()) -> (AllQuestsAreClaimedText, ThereIsSingleAvailableQuestText:Text, ReturnToTheLobbyText, FlushToTheBattleText).ThereIsSingleAvailableQuestText
   )
   n12 = Event ShowNewBang!() -> (OutputDelegate, then, bEnabled:bool) {is=title K2Node_Event}
   > Set Visibility(
    self=Get NewBang() -> (NewBang:Common Border Object Reference)
    InVisibility=Select(Option 0=Collapsed, Option 1=HitTestInvisible, Index=@n12.bEnabled)
   )
   Event PopulateAllQuestsAreCompletedButSomeNotClaimed() {is=title K2Node_Event}
   > Populate Was Demand During the Battle()
   n13 = Event Destruct()
   Deactivate Control Bus Mix!(execute=[@n5.then, @n13.then], Mix:Sound Control Bus Mix Object Reference=&CBM_S28_UI_Winterfest) -> (then)
   n14 = Event PopulateRelogin!() -> (OutputDelegate, then, bInFrontend:bool) {is=title K2Node_Event}
   n15 = Break Challenge Bundle Gift Card Display Data!(ChallengeBundleGiftCardDisplayData=Get DisplayData()) -> (ReturnToTheLobbyText, FlushToTheBattleText)
   > SetText(
    self=Get RichText_Description()
    InText=Select(Option 0:"Text (by ref)"=@n15, Option 1:"Text (by ref)"=@n15.FlushToTheBattleText, Index=@n14.bInFrontend) -> (ReturnValue:"Text (by ref)")
   )
   > SetActiveWidget(
    self=Get ContentSwitcher()
    Widget=Get GiftReadyNotAccessible_Container() -> (GiftReadyNotAccessible_Container:Vertical Box Object Reference)
   )
   n16 = Event PopulateAllQuestsAreClaimedBundleIsNotCompleted!() -> (OutputDelegate, then, TimeRemainingForNextQuest:"Timespan Structure (by ref)") {is=title K2Node_Event}
   > SetText(
    self=Get RichText_Description()
    InText=Break Challenge Bundle Gift Card Display Data!(ChallengeBundleGiftCardDisplayData=Get DisplayData()) -> (HasNoAvailableQuestsText:Text, AllQuestsAreClaimedText:Text, ReturnToTheLobbyText, FlushToTheBattleText)
   )
   > SetActiveWidget(self=Get ContentSwitcher(), Widget=Get NextGift_Container() -> (NextGift_Container:Vertical Box Object Reference))
   > "SetText (Text)"(
    self=Get Text_TimeForNextDemand() -> (Text_TimeForNextDemand:Fort Event Date Time Text Block Object Reference)
    InText=GetTimespanAsText!(
     Timespan:"Timespan Structure (by ref)"=@n16.TimeRemainingForNextQuest
     bShowHoursWhenZero:bool=false
     bShowSeconds:bool=false
     FormatType:ETimespanAsTextFormat Enum=DaysHoursMinutesSeconds
     UnitDisplayType:ETimespanUnitDisplayFormat Enum=Abbreviated
    ) -> (ReturnValue:Text)
   )
   MakeLiteralGameplayTagContainer!(Value:Gameplay Tag Container Structure=(GameplayTags=((TagName="Athena.Quests.Winterfest.Frontend")))) -> (ReturnValue:Gameplay Tag Container Structure)
   Get GiftCompleted_Container()
  ]}
 PopulateWasDemandDuringTheBattle {nodes [
   PopulateWasDemandDuringTheBattle() {is=title K2Node_FunctionEntry}
   > SetText(
    self=Get RichText_Description()
    InText=Break Challenge Bundle Gift Card Display Data!(ChallengeBundleGiftCardDisplayData=Get DisplayData()) -> (
     AllQuestsAreClaimedText
     ThereIsSingleAvailableQuestText:Text
     ReturnToTheLobbyText
     FlushToTheBattleText
     JustCompletedAllQuestsInBattleText:Text
    ).JustCompletedAllQuestsInBattleText
   )
   > SetActiveWidget(self=Get ContentSwitcher(), Widget=Get GiftClaimed_Container() -> (GiftClaimed_Container:Vertical Box Object Reference))
  ]}
 Sequencer Events {nodes [PlaySelected() {is=title K2Node_CustomEvent}]}
]
variables []
widgets {
 info {inheritedWidgetCount=0, parentClass=null, rootWidgetClass=&Overlay, widgetCount=58}
 widgets [
  Overlay_0 {widgetClassPath=&Overlay, widget=Overlay_0}
  Vignette {
   widgetClassPath=&Image
   bIsVariable=true
   parent=Overlay_0
   widget=Vignette
   properties {brush {
     bIsDynamicallyLoaded=false
     drawAs=Image
     imageType=NoImage
     mirroring=NoMirror
     resourceName=None
     resourceObject=/BattlePassS28_Quests/UI/Materials/M_BoxScaleGlow_WinterfestVignette
     tiling=NoTile
     imageSize=<32 32>
     margin=<ltrb 0 0 0 0>
     outlineSettings {
      bUseBrushTransparency=false
      roundingType=HalfHeightRadius
      width=0
      color {colorUseRule=UseColor_Specified, specifiedColor=<rgba 0 0 0 0>}
      cornerRadii=<0 0 0 0>
     }
     tintColor {colorUseRule=UseColor_Specified, specifiedColor=<rgba 1 1 1 0.7>}
     uVRegion {bIsValid=false, max=<0 0>, min=<0 0>}
    }}
   slot {class=OverlaySlot, properties {horizontalAlignment=HAlign_Fill, verticalAlignment=VAlign_Fill, padding=<ltrb -80 -30 -810 0>}}
  }
  SizeBox_StyleSize {
   widgetClassPath=&SizeBox
   bIsVariable=true
   parent=Overlay_0
   widget=SizeBox_StyleSize
   properties {
    bOverride_MinDesiredHeight=true
    bOverride_WidthOverride=true
    heightOverride=144
    maxDesiredHeight=105
    minDesiredHeight=160
    widthOverride=900
   }
   slot {class=OverlaySlot, properties {horizontalAlignment=HAlign_Fill, verticalAlignment=VAlign_Fill}}
  }
  Overlay_4 {
   widgetClassPath=&Overlay
   parent=SizeBox_StyleSize
   widget=Overlay_4
   slot {class=SizeBoxSlot, properties {padding=<ltrb 0 105 0 0>}}
  }
  Character {
   widgetClassPath=&Image
   bIsVariable=true
   parent=Overlay_4
   widget=Character
   properties {
    visibility=SelfHitTestInvisible
    brush {
     bIsDynamicallyLoaded=false
     drawAs=Image
     imageType=NoImage
     mirroring=NoMirror
     resourceName=None
     resourceObject=/BattlePassS28_Quests/UI/Materials/M_UI_Winterfest_Character
     tiling=NoTile
     imageSize=<400 400>
     margin=<ltrb 0 0 0 0>
     outlineSettings {
      bUseBrushTransparency=false
      roundingType=HalfHeightRadius
      width=0
      color {colorUseRule=UseColor_Specified, specifiedColor=<rgba 0 0 0 0>}
      cornerRadii=<0 0 0 0>
     }
     tintColor {colorUseRule=UseColor_Specified, specifiedColor=<rgba 1 1 1 1>}
     uVRegion {bIsValid=false, max=<0 0>, min=<0 0>}
    }
    renderTransform {angle=0, scale=<-2 2>, shear=<0 0>, translation=<474 60>}
   }
   slot {class=OverlaySlot, properties {horizontalAlignment=HAlign_Right, verticalAlignment=VAlign_Center}}
  }
  SizeBox_1 {
   widgetClassPath=&SizeBox
   parent=Overlay_4
   widget=SizeBox_1
   properties {bOverride_MaxDesiredWidth=true, bOverride_MinDesiredWidth=true, maxDesiredWidth=550, minDesiredWidth=550}
   slot {class=OverlaySlot, properties {horizontalAlignment=HAlign_Fill, verticalAlignment=VAlign_Fill, padding=<ltrb 50 0 0 0>}}
  }
  VerticalBox {widgetClassPath=&VerticalBox, parent=SizeBox_1, widget=VerticalBox, slot {class=SizeBoxSlot}}
  RichText_Description {
   widgetClassPath=&CommonRichTextBlock
   bIsVariable=true
   parent=VerticalBox
   widget=RichText_Description
   properties {
    autoWrapText=true
    bOverrideDefaultStyle=true
    justification=Center
    mobileTextBlockScale=1.3
    text="The next Ship It! Express gift is on its way!"
    defaultTextStyleOverride {
     overflowPolicy=Clip
     transformPolicy=None
     colorAndOpacity {colorUseRule=UseColor_Specified, specifiedColor=<rgba 1 1 1 1>}
     font {
      bForceMonospaced=false
      bMaterialIsStencil=false
      fontMaterial=null
      fontObject=&HeadingNow
      letterSpacing=0
      monospacedWidth=1
      size=21
      skewAmount=0
      typefaceFontName=75 Medium
      outlineSettings {
       bApplyOutlineToDropShadows=false
       bMiteredCorners=false
       bSeparateFillAlpha=false
       outlineMaterial=null
       outlineSize=0
       outlineColor=<rgba 0 0 0 1>
      }
     }
     highlightColor {colorUseRule=UseColor_Specified, specifiedColor=<rgba 0 0 0 1>}
     highlightShape {
      bIsDynamicallyLoaded=false
      drawAs=Image
      imageType=NoImage
      mirroring=NoMirror
      resourceName=None
      resourceObject=null
      tiling=NoTile
      imageSize=<32 32>
      margin=<ltrb 0 0 0 0>
      outlineSettings {
       bUseBrushTransparency=false
       roundingType=HalfHeightRadius
       width=0
       color {colorUseRule=UseColor_Specified, specifiedColor=<rgba 0 0 0 0>}
       cornerRadii=<0 0 0 0>
      }
      tintColor {colorUseRule=UseColor_Specified, specifiedColor=<rgba 1 1 1 1>}
      uVRegion {bIsValid=false, max=<0 0>, min=<0 0>}
     }
     selectedBackgroundColor {colorUseRule=UseColor_Foreground, specifiedColor=<rgba 1 0 1 1>}
     shadowColorAndOpacity=<rgba 0 0 0 1>
     shadowOffset=<0 0>
     strikeBrush {
      bIsDynamicallyLoaded=false
      drawAs=Image
      imageType=NoImage
      mirroring=NoMirror
      resourceName=None
      resourceObject=null
      tiling=NoTile
      imageSize=<32 32>
      margin=<ltrb 0 0 0 0>
      outlineSettings {
       bUseBrushTransparency=false
       roundingType=HalfHeightRadius
       width=0
       color {colorUseRule=UseColor_Specified, specifiedColor=<rgba 0 0 0 0>}
       cornerRadii=<0 0 0 0>
      }
      tintColor {colorUseRule=UseColor_Specified, specifiedColor=<rgba 1 1 1 1>}
      uVRegion {bIsValid=false, max=<0 0>, min=<0 0>}
     }
     underlineBrush {
      bIsDynamicallyLoaded=false
      drawAs=Image
      imageType=NoImage
      mirroring=NoMirror
      resourceName=None
      resourceObject=null
      tiling=NoTile
      imageSize=<32 32>
      margin=<ltrb 0 0 0 0>
      outlineSettings {
       bUseBrushTransparency=false
       roundingType=HalfHeightRadius
       width=0
       color {colorUseRule=UseColor_Specified, specifiedColor=<rgba 0 0 0 0>}
       cornerRadii=<0 0 0 0>
      }
      tintColor {colorUseRule=UseColor_Specified, specifiedColor=<rgba 1 1 1 1>}
      uVRegion {bIsValid=false, max=<0 0>, min=<0 0>}
     }
    }
   }
   slot {class=VerticalBoxSlot, properties {verticalAlignment=VAlign_Center, size {sizeRule=Fill, value=1}}}
  }
  WordsFlourish {
   widgetClassPath=&Image
   bIsVariable=true
   parent=VerticalBox
   widget=WordsFlourish
   properties {brush {
     bIsDynamicallyLoaded=false
     drawAs=Image
     imageType=NoImage
     mirroring=NoMirror
     resourceName=None
     resourceObject=/BattlePassS28_Quests/UI/Materials/M_UI_Winterfest_Flourish
     tiling=NoTile
     imageSize=<680 124>
     margin=<ltrb 0 0 0 0>
     outlineSettings {
      bUseBrushTransparency=false
      roundingType=HalfHeightRadius
      width=0
      color {colorUseRule=UseColor_Specified, specifiedColor=<rgba 0 0 0 0>}
      cornerRadii=<0 0 0 0>
     }
     tintColor {colorUseRule=UseColor_Specified, specifiedColor=<rgba 1 1 1 1>}
     uVRegion {bIsValid=false, max=<0 0>, min=<0 0>}
    }}
   slot {class=VerticalBoxSlot, properties {horizontalAlignment=HAlign_Center, verticalAlignment=VAlign_Center, padding=<ltrb 0 30 0 30>}}
  }
  ContentSwitcher {
   widgetClassPath=&CommonVisibilitySwitcher
   bIsVariable=true
   parent=VerticalBox
   widget=ContentSwitcher
   slot {class=VerticalBoxSlot, properties {padding=<ltrb 0 16 0 0>}}
  }
  NextGift_Container {
   widgetClassPath=&VerticalBox
   bIsVariable=true
   parent=ContentSwitcher
   widget=NextGift_Container
   slot {class=CommonVisibilitySwitcherSlot, properties {horizontalAlignment=HAlign_Center, verticalAlignment=VAlign_Center}}
  }
  Overlay {
   widgetClassPath=&Overlay
   parent=NextGift_Container
   widget=Overlay
   slot {class=VerticalBoxSlot, properties {horizontalAlignment=HAlign_Center}}
  }
  ShipItParcel {
   widgetClassPath=&Image
   parent=Overlay
   widget=ShipItParcel
   properties {
    visibility=SelfHitTestInvisible
    brush {
     bIsDynamicallyLoaded=false
     drawAs=Image
     imageType=NoImage
     mirroring=NoMirror
     resourceName=None
     resourceObject=/BattlePassS28_Quests/UI/Materials/M_UI_Winterfest_ShipItParcel
     tiling=NoTile
     imageSize=<260 260>
     margin=<ltrb 0 0 0 0>
     outlineSettings {
      bUseBrushTransparency=false
      roundingType=HalfHeightRadius
      width=0
      color {colorUseRule=UseColor_Specified, specifiedColor=<rgba 0 0 0 0>}
      cornerRadii=<0 0 0 0>
     }
     tintColor {colorUseRule=UseColor_Specified, specifiedColor=<rgba 0.276 0.276 0.276 1>}
     uVRegion {bIsValid=false, max=<0 0>, min=<0 0>}
    }
   }
   slot {class=OverlaySlot}
  }
  ShipItParcel_Outline {
   widgetClassPath=&Image
   parent=Overlay
   widget=ShipItParcel_Outline
   properties {
    visibility=SelfHitTestInvisible
    brush {
     bIsDynamicallyLoaded=false
     drawAs=Image
     imageType=NoImage
     mirroring=NoMirror
     resourceName=None
     resourceObject=/BattlePassS28_Quests/UI/Textures/T_UI_Winterfest_ShipItParcelOutline
     tiling=NoTile
     imageSize=<260 260>
     margin=<ltrb 0 0 0 0>
     outlineSettings {
      bUseBrushTransparency=false
      roundingType=HalfHeightRadius
      width=0
      color {colorUseRule=UseColor_Specified, specifiedColor=<rgba 0 0 0 0>}
      cornerRadii=<0 0 0 0>
     }
     tintColor {colorUseRule=UseColor_Specified, specifiedColor=<rgba 1 1 1 1>}
     uVRegion {bIsValid=false, max=<0 0>, min=<0 0>}
    }
   }
   slot {class=OverlaySlot}
  }
  HorizontalBox_82 {
   widgetClassPath=&HorizontalBox
   parent=NextGift_Container
   widget=HorizontalBox_82
   slot {class=VerticalBoxSlot, properties {padding=<ltrb 0 50 0 0>}}
  }
  SBTimer {
   widgetClassPath=&SizeBox
   parent=HorizontalBox_82
   widget=SBTimer
   properties {bOverride_HeightOverride=true, bOverride_WidthOverride=true, heightOverride=84, widthOverride=84}
   slot {class=HorizontalBoxSlot, properties {horizontalAlignment=HAlign_Center, verticalAlignment=VAlign_Center}}
  }
  TimerImage {
   widgetClassPath=&Image
   parent=SBTimer
   widget=TimerImage
   properties {
    brush {
     bIsDynamicallyLoaded=false
     drawAs=Image
     imageType=NoImage
     mirroring=NoMirror
     resourceName=None
     resourceObject=/Game/Athena/UI/Frontend/ActivityBrowser/Art/M_UI_Timer_Icon_ShopOffer
     tiling=NoTile
     imageSize=<32 32>
     margin=<ltrb 0 0 0 0>
     outlineSettings {
      bUseBrushTransparency=false
      roundingType=HalfHeightRadius
      width=0
      color {colorUseRule=UseColor_Specified, specifiedColor=<rgba 0 0 0 0>}
      cornerRadii=<0 0 0 0>
     }
     tintColor {colorUseRule=UseColor_Specified, specifiedColor=<rgba 1 1 1 1>}
     uVRegion {bIsValid=false, max=<0 0>, min=<0 0>}
    }
    colorAndOpacity=<rgba 0.0012 0.0242 0.1441 1>
   }
   slot {class=SizeBoxSlot}
  }
  VerticalBox_212 {
   widgetClassPath=&VerticalBox
   parent=HorizontalBox_82
   widget=VerticalBox_212
   slot {class=HorizontalBoxSlot, properties {size {sizeRule=Fill, value=1}}}
  }
  HorizontalBox_0 {
   widgetClassPath=&HorizontalBox
   parent=VerticalBox_212
   widget=HorizontalBox_0
   slot {class=VerticalBoxSlot, properties {horizontalAlignment=HAlign_Left, verticalAlignment=VAlign_Center}}
  }
  Text_Content {
   widgetClassPath=&CommonTextBlock
   parent=HorizontalBox_0
   widget=Text_Content
   properties {
    mobileFontSizeMultiplier=1.2
    scrollStyle=/FortUILibrary/Styles/Typography/Foundation/TextScrollStyle-NoFade.TextScrollStyle-NoFade_C
    text=Shop Refresh
    textTransformPolicy=ToUpper
    colorAndOpacity {colorUseRule=UseColor_Specified, specifiedColor=<rgba 0.0012 0.0242 0.1441 1>}
    font {
     bForceMonospaced=false
     bMaterialIsStencil=false
     fontMaterial=null
     fontObject=&HeadingNow
     letterSpacing=0
     monospacedWidth=1
     size=22.5
     skewAmount=0
     typefaceFontName=75 Medium
     outlineSettings {
      bApplyOutlineToDropShadows=false
      bMiteredCorners=false
      bSeparateFillAlpha=false
      outlineMaterial=null
      outlineSize=0
      outlineColor=<rgba 0 0 0 1>
     }
    }
    shadowOffset=<0 0>
   }
   slot {class=HorizontalBoxSlot, properties {padding=<ltrb 0 0 6 0>}}
  }
  SizeBox_232 {
   widgetClassPath=&SizeBox
   parent=HorizontalBox_0
   widget=SizeBox_232
   properties {bOverride_HeightOverride=true, bOverride_WidthOverride=true, heightOverride=30, visibility=Collapsed, widthOverride=30}
   slot {class=HorizontalBoxSlot, properties {verticalAlignment=VAlign_Center}}
  }
  Overlay_159 {widgetClassPath=&Overlay, parent=SizeBox_232, widget=Overlay_159, slot {class=SizeBoxSlot}}
  CommonHardwareVisibilityBorder_446 {
   widgetClassPath=&CommonHardwareVisibilityBorder
   parent=Overlay_159
   widget=CommonHardwareVisibilityBorder_446
   properties {visibilityQuery {
     autoDescription=" ANY( Input.MouseAndKeyboard, Input.Touch )"
     tokenStreamVersion=0
     userDescription=""
     queryTokenStream [0,1,1,2,0,1]
     tagDictionary [{tagName=Input.MouseAndKeyboard}, {tagName=Input.Touch}]
    }}
   slot {class=OverlaySlot, properties {horizontalAlignment=HAlign_Fill, verticalAlignment=VAlign_Fill}}
  }
  ScaleBox_575 {
   widgetClassPath=&ScaleBox
   parent=CommonHardwareVisibilityBorder_446
   widget=ScaleBox_575
   slot {class=BorderSlot, properties {padding=<ltrb 0 0 0 0>}}
  }
  QuestionMarkBorder {
   widgetClassPath=&Border
   parent=ScaleBox_575
   widget=QuestionMarkBorder
   properties {
    background {
     bIsDynamicallyLoaded=false
     drawAs=Image
     imageType=NoImage
     mirroring=NoMirror
     resourceName=None
     resourceObject=/Game/UI/Materials/M_UICircle
     tiling=NoTile
     imageSize=<24 24>
     margin=<ltrb 0 0 0 0>
     outlineSettings {
      bUseBrushTransparency=false
      roundingType=HalfHeightRadius
      width=0
      color {colorUseRule=UseColor_Specified, specifiedColor=<rgba 0 0 0 0>}
      cornerRadii=<0 0 0 0>
     }
     tintColor {colorUseRule=UseColor_Specified, specifiedColor=<rgba 0 0 0 1>}
     uVRegion {bIsValid=false, max=<0 0>, min=<0 0>}
    }
    padding=<ltrb 0 0 0 0>
   }
   slot {class=ScaleBoxSlot, properties {horizontalAlignment=HAlign_Fill, verticalAlignment=VAlign_Fill}}
  }
  OutlineBorder {
   widgetClassPath=&Border
   parent=QuestionMarkBorder
   widget=OutlineBorder
   properties {
    verticalAlignment=VAlign_Center
    visibility=SelfHitTestInvisible
    background {
     bIsDynamicallyLoaded=false
     drawAs=Image
     imageType=NoImage
     mirroring=NoMirror
     resourceName=None
     resourceObject=/FortniteCoreUI/InGame/Interrogation/Materials/M_UI_QTECircle_TargetRing
     tiling=NoTile
     imageSize=<28 28>
     margin=<ltrb 0 0 0 0>
     outlineSettings {
      bUseBrushTransparency=false
      roundingType=HalfHeightRadius
      width=0
      color {colorUseRule=UseColor_Specified, specifiedColor=<rgba 0 0 0 0>}
      cornerRadii=<0 0 0 0>
     }
     tintColor {colorUseRule=UseColor_Specified, specifiedColor=<rgba 1 1 1 1>}
     uVRegion {bIsValid=false, max=<0 0>, min=<0 0>}
    }
    brushColor=<rgba 0.004 0.0704 0.4233 1>
    padding=<ltrb 0 0 0 0>
   }
   slot {class=BorderSlot, properties {padding=<ltrb 0 0 0 0>}}
  }
  QuestionMark {
   widgetClassPath=&CommonTextBlock
   parent=OutlineBorder
   widget=QuestionMark
   properties {
    justification=Center
    text=?
    font {
     bForceMonospaced=false
     bMaterialIsStencil=false
     fontMaterial=null
     fontObject=&BurbankSmall
     letterSpacing=0
     monospacedWidth=1
     size=11
     skewAmount=0
     typefaceFontName=Bold
     outlineSettings {
      bApplyOutlineToDropShadows=false
      bMiteredCorners=false
      bSeparateFillAlpha=false
      outlineMaterial=null
      outlineSize=0
      outlineColor=<rgba 0 0 0 1>
     }
    }
    renderTransform {angle=12, scale=<1 1>, shear=<0 0>, translation=<0 0>}
    shadowOffset=<0 0>
   }
   slot {class=BorderSlot, properties {verticalAlignment=VAlign_Center, padding=<ltrb 0 0 0 0>}}
  }
  CommonHardwareVisibilityBorder {
   widgetClassPath=&CommonHardwareVisibilityBorder
   parent=Overlay_159
   widget=CommonHardwareVisibilityBorder
   properties {visibilityQuery {
     autoDescription=" ANY( Input.Gamepad )"
     tokenStreamVersion=0
     userDescription=""
     queryTokenStream [0,1,1,1,0]
     tagDictionary [{tagName=Input.Gamepad}]
    }}
   slot {class=OverlaySlot, properties {horizontalAlignment=HAlign_Fill, verticalAlignment=VAlign_Fill}}
  }
  InputActionWidget {
   widgetClassPath=&CommonActionWidget
   parent=CommonHardwareVisibilityBorder
   widget=InputActionWidget
   properties {progressMaterialBrush {
     bIsDynamicallyLoaded=false
     drawAs=Image
     imageType=NoImage
     mirroring=NoMirror
     resourceName=None
     resourceObject=null
     tiling=NoTile
     imageSize=<64 64>
     margin=<ltrb 0 0 0 0>
     outlineSettings {
      bUseBrushTransparency=false
      roundingType=HalfHeightRadius
      width=0
      color {colorUseRule=UseColor_Specified, specifiedColor=<rgba 0 0 0 0>}
      cornerRadii=<0 0 0 0>
     }
     tintColor {colorUseRule=UseColor_Specified, specifiedColor=<rgba 1 1 1 1>}
     uVRegion {bIsValid=false, max=<0 0>, min=<0 0>}
    }}
   slot {class=BorderSlot, properties {padding=<ltrb 0 0 0 0>}}
  }
  Text_TimeForNextDemand {
   widgetClassPath=/Script/FortniteUI.FortEventDateTimeTextBlock
   bIsVariable=true
   parent=VerticalBox_212
   widget=Text_TimeForNextDemand
   properties {
    mobileFontSizeMultiplier=1.2
    text=0:00
    textTransformPolicy=ToUpper
    colorAndOpacity {colorUseRule=UseColor_Specified, specifiedColor=<rgba 0.004 0.1683 0.9301 1>}
    font {
     bForceMonospaced=false
     bMaterialIsStencil=false
     fontMaterial=null
     fontObject=&HeadingNow
     letterSpacing=0
     monospacedWidth=1
     size=32.25
     skewAmount=0
     typefaceFontName=86 Bold Italic
     outlineSettings {
      bApplyOutlineToDropShadows=false
      bMiteredCorners=false
      bSeparateFillAlpha=false
      outlineMaterial=null
      outlineSize=0
      outlineColor=<rgba 0 0 0 1>
     }
    }
    shadowOffset=<0 0>
   }
   slot {class=VerticalBoxSlot, properties {verticalAlignment=VAlign_Center}}
  }
  GiftReadyNotAccessible_Container {
   widgetClassPath=&VerticalBox
   bIsVariable=true
   parent=ContentSwitcher
   widget=GiftReadyNotAccessible_Container
   slot {class=CommonVisibilitySwitcherSlot}
  }
  Present_Container {widgetClassPath=&Overlay, parent=GiftReadyNotAccessible_Container, widget=Present_Container, slot {class=VerticalBoxSlot}}
  Present {
   widgetClassPath=&Image
   bIsVariable=true
   parent=Present_Container
   widget=Present
   properties {brush {
     bIsDynamicallyLoaded=false
     drawAs=Image
     imageType=NoImage
     mirroring=NoMirror
     resourceName=None
     resourceObject=/BattlePassS28_Quests/UI/Materials/M_UI_Winterfest_Present
     tiling=NoTile
     imageSize=<260 260>
     margin=<ltrb 0 0 0 0>
     outlineSettings {
      bUseBrushTransparency=false
      roundingType=HalfHeightRadius
      width=0
      color {colorUseRule=UseColor_Specified, specifiedColor=<rgba 0 0 0 0>}
      cornerRadii=<0 0 0 0>
     }
     tintColor {colorUseRule=UseColor_Specified, specifiedColor=<rgba 1 1 1 1>}
     uVRegion {bIsValid=false, max=<0 0>, min=<0 0>}
    }}
   slot {class=OverlaySlot, properties {horizontalAlignment=HAlign_Center, verticalAlignment=VAlign_Center}}
  }
  WBP_S28_Winterfest_PresentTag {
   widgetClassPath=&/WBP_S28_Winterfest_PresentTag
   bIsVariable=true
   parent=GiftReadyNotAccessible_Container
   widget=WBP_S28_Winterfest_PresentTag
   slot {class=VerticalBoxSlot, properties {horizontalAlignment=HAlign_Center, verticalAlignment=VAlign_Top, padding=<ltrb 0 60 0 0>}}
  }
  GiftClaimed_Container {
   widgetClassPath=&VerticalBox
   bIsVariable=true
   parent=ContentSwitcher
   widget=GiftClaimed_Container
   slot {class=CommonVisibilitySwitcherSlot}
  }
  Present_Container_2 {widgetClassPath=&Overlay, parent=GiftClaimed_Container, widget=Present_Container_2, slot {class=VerticalBoxSlot}}
  Present_2 {
   widgetClassPath=&Image
   bIsVariable=true
   parent=Present_Container_2
   widget=Present_2
   properties {brush {
     bIsDynamicallyLoaded=false
     drawAs=Image
     imageType=NoImage
     mirroring=NoMirror
     resourceName=None
     resourceObject=/BattlePassS28_Quests/UI/Materials/M_UI_Winterfest_PresentClaimed
     tiling=NoTile
     imageSize=<260 260>
     margin=<ltrb 0 0 0 0>
     outlineSettings {
      bUseBrushTransparency=false
      roundingType=HalfHeightRadius
      width=0
      color {colorUseRule=UseColor_Specified, specifiedColor=<rgba 0 0 0 0>}
      cornerRadii=<0 0 0 0>
     }
     tintColor {colorUseRule=UseColor_Specified, specifiedColor=<rgba 1 1 1 1>}
     uVRegion {bIsValid=false, max=<0 0>, min=<0 0>}
    }}
   slot {class=OverlaySlot, properties {horizontalAlignment=HAlign_Center, verticalAlignment=VAlign_Center}}
  }
  PresentTick_2 {
   widgetClassPath=&Image
   bIsVariable=true
   parent=Present_Container_2
   widget=PresentTick_2
   properties {
    brush {
     bIsDynamicallyLoaded=false
     drawAs=Image
     imageType=NoImage
     mirroring=NoMirror
     resourceName=None
     resourceObject=/BattlePassS28_Quests/UI/Materials/M_UI_Winterfest_PresentTick
     tiling=NoTile
     imageSize=<86 86>
     margin=<ltrb 0 0 0 0>
     outlineSettings {
      bUseBrushTransparency=false
      roundingType=HalfHeightRadius
      width=0
      color {colorUseRule=UseColor_Specified, specifiedColor=<rgba 0 0 0 0>}
      cornerRadii=<0 0 0 0>
     }
     tintColor {colorUseRule=UseColor_Specified, specifiedColor=<rgba 1 1 1 1>}
     uVRegion {bIsValid=false, max=<0 0>, min=<0 0>}
    }
    renderTransform {angle=0, scale=<1 1>, shear=<0 0>, translation=<93 -79>}
   }
   slot {class=OverlaySlot, properties {horizontalAlignment=HAlign_Center, verticalAlignment=VAlign_Center}}
  }
  WBP_S28_Winterfest_PresentTagClaimed {
   widgetClassPath=&/WBP_S28_Winterfest_PresentTagClaimed
   bIsVariable=true
   parent=GiftClaimed_Container
   widget=WBP_S28_Winterfest_PresentTagClaimed
   slot {class=VerticalBoxSlot, properties {horizontalAlignment=HAlign_Center, verticalAlignment=VAlign_Top, padding=<ltrb 0 60 0 0>}}
  }
  GiftReady_Container {
   widgetClassPath=&VerticalBox
   bIsVariable=true
   parent=ContentSwitcher
   widget=GiftReady_Container
   slot {class=CommonVisibilitySwitcherSlot}
  }
  Present_Container_1 {widgetClassPath=&Overlay, parent=GiftReady_Container, widget=Present_Container_1, slot {class=VerticalBoxSlot}}
  Present_1 {
   widgetClassPath=&Image
   bIsVariable=true
   parent=Present_Container_1
   widget=Present_1
   properties {brush {
     bIsDynamicallyLoaded=false
     drawAs=Image
     imageType=NoImage
     mirroring=NoMirror
     resourceName=None
     resourceObject=&M_UI_Winterfest_Present
     tiling=NoTile
     imageSize=<260 260>
     margin=<ltrb 0 0 0 0>
     outlineSettings {
      bUseBrushTransparency=false
      roundingType=HalfHeightRadius
      width=0
      color {colorUseRule=UseColor_Specified, specifiedColor=<rgba 0 0 0 0>}
      cornerRadii=<0 0 0 0>
     }
     tintColor {colorUseRule=UseColor_Specified, specifiedColor=<rgba 1 1 1 1>}
     uVRegion {bIsValid=false, max=<0 0>, min=<0 0>}
    }}
   slot {class=OverlaySlot, properties {horizontalAlignment=HAlign_Center, verticalAlignment=VAlign_Center}}
  }
  ButtonContainer {
   widgetClassPath=&Overlay
   bIsVariable=true
   parent=GiftReady_Container
   widget=ButtonContainer
   slot {class=VerticalBoxSlot, properties {horizontalAlignment=HAlign_Center, padding=<ltrb 20 68 0 0>}}
  }
  WBP_UIKit_Button_Regular {
   widgetClassPath=&WBP_UIKit_Button_Regular
   bIsVariable=true
   parent=ButtonContainer
   widget=WBP_UIKit_Button_Regular
   properties {
    bHideInputActionWithKeyboard=true
    bIsFocusable=false
    buttonSize=Small
    minHeight=56
    text=OPEN
    triggeringInputAction {dataTable=/FortUILibrary/Input/AthenaMiniMapActionDataTable, rowName=OpenGift}
   }
   slot {class=OverlaySlot, properties {horizontalAlignment=HAlign_Right, verticalAlignment=VAlign_Center, padding=<ltrb 0 0 46 0>}}
  }
  PendingAmountBlock {
   widgetClassPath=&Overlay
   bIsVariable=true
   parent=ButtonContainer
   widget=PendingAmountBlock
   properties {renderTransform {angle=0, scale=<1 1>, shear=<0 0>, translation=<10 0>}}
   slot {class=OverlaySlot, properties {horizontalAlignment=HAlign_Right, verticalAlignment=VAlign_Center}}
  }
  Image_717 {
   widgetClassPath=&Image
   bIsVariable=true
   parent=PendingAmountBlock
   widget=Image_717
   properties {
    brush {
     bIsDynamicallyLoaded=false
     drawAs=Image
     imageType=NoImage
     mirroring=NoMirror
     resourceName=None
     resourceObject=/BattlePassS28_Quests/UI/Materials/MI_UI_Winterfest_Tag_Master
     tiling=NoTile
     imageSize=<70 42>
     margin=<ltrb 0 0 0 0>
     outlineSettings {
      bUseBrushTransparency=false
      roundingType=HalfHeightRadius
      width=0
      color {colorUseRule=UseColor_Specified, specifiedColor=<rgba 0 0 0 0>}
      cornerRadii=<0 0 0 0>
     }
     tintColor {colorUseRule=UseColor_Specified, specifiedColor=<rgba 1 1 1 1>}
     uVRegion {bIsValid=false, max=<0 0>, min=<0 0>}
    }
    renderTransform {angle=0, scale=<-1 1>, shear=<0 0>, translation=<0 0>}
   }
   slot {class=OverlaySlot, properties {horizontalAlignment=HAlign_Fill, verticalAlignment=VAlign_Fill}}
  }
  Text_PendingGiftsAmount {
   widgetClassPath=&CommonTextBlock
   bIsVariable=true
   parent=PendingAmountBlock
   widget=Text_PendingGiftsAmount
   properties {
    text=x3
    font {
     bForceMonospaced=false
     bMaterialIsStencil=false
     fontMaterial=/FortniteCoreUI/Athena/HUD/Quests/Art/M_UI_Timer_Font_Quest
     fontObject=&HeadingNow
     letterSpacing=0
     monospacedWidth=1
     size=15.75
     skewAmount=0
     typefaceFontName=75 Medium
     outlineSettings {
      bApplyOutlineToDropShadows=false
      bMiteredCorners=false
      bSeparateFillAlpha=false
      outlineMaterial=null
      outlineSize=0
      outlineColor=<rgba 0 0 0 1>
     }
    }
    shadowOffset=<0 0>
   }
   slot {class=OverlaySlot, properties {horizontalAlignment=HAlign_Center, verticalAlignment=VAlign_Center}}
  }
  GiftCompleted_Container {
   widgetClassPath=&Overlay
   bIsVariable=true
   parent=ContentSwitcher
   widget=GiftCompleted_Container
   slot {class=CommonVisibilitySwitcherSlot, properties {verticalAlignment=VAlign_Center}}
  }
  GiftCompleted {
   widgetClassPath=&Image
   bIsVariable=true
   parent=GiftCompleted_Container
   widget=GiftCompleted
   properties {
    visibility=SelfHitTestInvisible
    brush {
     bIsDynamicallyLoaded=false
     drawAs=Image
     imageType=NoImage
     mirroring=NoMirror
     resourceName=None
     resourceObject=&M_UI_Winterfest_PresentTick
     tiling=NoTile
     imageSize=<310 310>
     margin=<ltrb 0 0 0 0>
     outlineSettings {
      bUseBrushTransparency=false
      roundingType=HalfHeightRadius
      width=0
      color {colorUseRule=UseColor_Specified, specifiedColor=<rgba 0 0 0 0>}
      cornerRadii=<0 0 0 0>
     }
     tintColor {colorUseRule=UseColor_Specified, specifiedColor=<rgba 1 1 1 1>}
     uVRegion {bIsValid=false, max=<0 0>, min=<0 0>}
    }
   }
   slot {class=OverlaySlot, properties {horizontalAlignment=HAlign_Center, verticalAlignment=VAlign_Center, padding=<ltrb 0 10 0 10>}}
  }
  GiftPreview_01 {
   widgetClassPath=&Image
   bIsVariable=true
   parent=GiftCompleted_Container
   widget=GiftPreview_01
   properties {
    visibility=SelfHitTestInvisible
    brush {
     bIsDynamicallyLoaded=false
     drawAs=Image
     imageType=NoImage
     mirroring=NoMirror
     resourceName=None
     resourceObject=/BattlePassS28_Quests/UI/Materials/M_UI_Winterfest_Reward_01
     tiling=NoTile
     imageSize=<220 220>
     margin=<ltrb 0 0 0 0>
     outlineSettings {
      bUseBrushTransparency=false
      roundingType=HalfHeightRadius
      width=0
      color {colorUseRule=UseColor_Specified, specifiedColor=<rgba 0 0 0 0>}
      cornerRadii=<0 0 0 0>
     }
     tintColor {colorUseRule=UseColor_Specified, specifiedColor=<rgba 1 1 1 1>}
     uVRegion {bIsValid=false, max=<0 0>, min=<0 0>}
    }
   }
   slot {class=OverlaySlot}
  }
  GiftPreview_02 {
   widgetClassPath=&Image
   bIsVariable=true
   parent=GiftCompleted_Container
   widget=GiftPreview_02
   properties {
    visibility=SelfHitTestInvisible
    brush {
     bIsDynamicallyLoaded=false
     drawAs=Image
     imageType=NoImage
     mirroring=NoMirror
     resourceName=None
     resourceObject=/BattlePassS28_Quests/UI/Materials/M_UI_Winterfest_Reward_02
     tiling=NoTile
     imageSize=<220 220>
     margin=<ltrb 0 0 0 0>
     outlineSettings {
      bUseBrushTransparency=false
      roundingType=HalfHeightRadius
      width=0
      color {colorUseRule=UseColor_Specified, specifiedColor=<rgba 0 0 0 0>}
      cornerRadii=<0 0 0 0>
     }
     tintColor {colorUseRule=UseColor_Specified, specifiedColor=<rgba 1 1 1 1>}
     uVRegion {bIsValid=false, max=<0 0>, min=<0 0>}
    }
    renderTransform {angle=0, scale=<1 1>, shear=<0 0>, translation=<0 110>}
   }
   slot {class=OverlaySlot, properties {verticalAlignment=VAlign_Bottom, padding=<ltrb 32 0 0 0>}}
  }
  GiftPreview_03 {
   widgetClassPath=&Image
   bIsVariable=true
   parent=GiftCompleted_Container
   widget=GiftPreview_03
   properties {
    visibility=SelfHitTestInvisible
    brush {
     bIsDynamicallyLoaded=false
     drawAs=Image
     imageType=NoImage
     mirroring=NoMirror
     resourceName=None
     resourceObject=/BattlePassS28_Quests/UI/Materials/M_UI_Winterfest_Reward_03
     tiling=NoTile
     imageSize=<220 220>
     margin=<ltrb 0 0 0 0>
     outlineSettings {
      bUseBrushTransparency=false
      roundingType=HalfHeightRadius
      width=0
      color {colorUseRule=UseColor_Specified, specifiedColor=<rgba 0 0 0 0>}
      cornerRadii=<0 0 0 0>
     }
     tintColor {colorUseRule=UseColor_Specified, specifiedColor=<rgba 1 1 1 1>}
     uVRegion {bIsValid=false, max=<0 0>, min=<0 0>}
    }
   }
   slot {class=OverlaySlot, properties {horizontalAlignment=HAlign_Right}}
  }
  GiftPreview_04 {
   widgetClassPath=&Image
   bIsVariable=true
   parent=GiftCompleted_Container
   widget=GiftPreview_04
   properties {
    visibility=SelfHitTestInvisible
    brush {
     bIsDynamicallyLoaded=false
     drawAs=Image
     imageType=NoImage
     mirroring=NoMirror
     resourceName=None
     resourceObject=/BattlePassS28_Quests/UI/Materials/M_UI_Winterfest_Reward_04
     tiling=NoTile
     imageSize=<220 220>
     margin=<ltrb 0 0 0 0>
     outlineSettings {
      bUseBrushTransparency=false
      roundingType=HalfHeightRadius
      width=0
      color {colorUseRule=UseColor_Specified, specifiedColor=<rgba 0 0 0 0>}
      cornerRadii=<0 0 0 0>
     }
     tintColor {colorUseRule=UseColor_Specified, specifiedColor=<rgba 1 1 1 1>}
     uVRegion {bIsValid=false, max=<0 0>, min=<0 0>}
    }
    renderTransform {angle=0, scale=<1 1>, shear=<0 0>, translation=<0 110>}
   }
   slot {class=OverlaySlot, properties {horizontalAlignment=HAlign_Right, verticalAlignment=VAlign_Bottom, padding=<ltrb 0 0 32 0>}}
  }
  Text_QuestID {
   widgetClassPath=&CommonTextBlock
   bIsVariable=true
   parent=Overlay_4
   widget=Text_QuestID
   properties {
    mobileFontSizeMultiplier=1.2
    text=Quest ID Watermark
    visibility=Collapsed
    colorAndOpacity {colorUseRule=UseColor_Specified, specifiedColor=<rgba 1 1 1 0.5>}
    font {
     bForceMonospaced=false
     bMaterialIsStencil=false
     fontMaterial=null
     fontObject=&NotoSans
     letterSpacing=0
     monospacedWidth=1
     size=10
     skewAmount=0
     typefaceFontName=NotoSans-Regular
     outlineSettings {
      bApplyOutlineToDropShadows=false
      bMiteredCorners=false
      bSeparateFillAlpha=false
      outlineMaterial=null
      outlineSize=0
      outlineColor=<rgba 0 0 0 1>
     }
    }
    shadowOffset=<0 0>
   }
   slot {class=OverlaySlot, properties {verticalAlignment=VAlign_Bottom, padding=<ltrb 32 0 0 4>}}
  }
  NewBang {
   widgetClassPath=&CommonBorder
   bIsVariable=true
   parent=Overlay_4
   widget=NewBang
   properties {
    horizontalAlignment=HAlign_Center
    background {
     bIsDynamicallyLoaded=false
     drawAs=Box
     imageType=NoImage
     mirroring=NoMirror
     resourceName=None
     resourceObject=/Game/Athena/UI/ReplayBrowser/M_Ui_Skewed_Vfor9slices_Bang
     tiling=NoTile
     imageSize=<32 32>
     margin=<ltrb 0.5 0 0.5 0>
     outlineSettings {
      bUseBrushTransparency=false
      roundingType=HalfHeightRadius
      width=0
      color {colorUseRule=UseColor_Specified, specifiedColor=<rgba 0 0 0 0>}
      cornerRadii=<0 0 0 0>
     }
     tintColor {colorUseRule=UseColor_Specified, specifiedColor=<rgba 0.9301 1 0.0097 1>}
     uVRegion {bIsValid=false, max=<0 0>, min=<0 0>}
    }
    padding=<ltrb 8 6 8 6>
   }
   slot {class=OverlaySlot, properties {horizontalAlignment=HAlign_Right, padding=<ltrb 2 -6 2 2>}}
  }
  CommonTextBlock_209 {
   widgetClassPath=&CommonTextBlock
   parent=NewBang
   widget=CommonTextBlock_209
   properties {
    lineHeightPercentage=0.7429
    mobileFontSizeMultiplier=1.2
    text="!"
    colorAndOpacity {colorUseRule=UseColor_Specified, specifiedColor=<rgba 0 0 0 1>}
    font {
     bForceMonospaced=false
     bMaterialIsStencil=false
     fontMaterial=null
     fontObject=&Burbank
     letterSpacing=0
     monospacedWidth=1
     size=18
     skewAmount=0
     typefaceFontName=BurbankRegular-Black
     outlineSettings {
      bApplyOutlineToDropShadows=false
      bMiteredCorners=false
      bSeparateFillAlpha=false
      outlineMaterial=null
      outlineSize=0
      outlineColor=<rgba 0 0 0 1>
     }
    }
    margin=<ltrb 0 -6 0 0>
    renderTransform {angle=0, scale=<1 1>, shear=<-12 0>, translation=<0 0>}
    shadowOffset=<0 0>
   }
   slot {class=BorderSlot, properties {horizontalAlignment=HAlign_Center, padding=<ltrb 8 6 8 6>}}
  }
  Image_Fade {
   widgetClassPath=&Image
   bIsVariable=true
   parent=Overlay_4
   widget=Image_Fade
   properties {
    visibility=HitTestInvisible
    brush {
     bIsDynamicallyLoaded=false
     drawAs=Box
     imageType=NoImage
     mirroring=NoMirror
     resourceName=None
     resourceObject=/FortniteCoreUI/HUD/MiniMap/Art/M_UI_Quests_ScreenSpaceFade_Rounded
     tiling=NoTile
     imageSize=<64 64>
     margin=<ltrb 0.5 0.5 0.5 0.5>
     outlineSettings {
      bUseBrushTransparency=false
      roundingType=HalfHeightRadius
      width=0
      color {colorUseRule=UseColor_Specified, specifiedColor=<rgba 0 0 0 0>}
      cornerRadii=<0 0 0 0>
     }
     tintColor {colorUseRule=UseColor_Specified, specifiedColor=<rgba 0 0.0056 0.1329 0.9>}
     uVRegion {bIsValid=false, max=<0 0>, min=<0 0>}
    }
   }
   slot {class=OverlaySlot, properties {horizontalAlignment=HAlign_Fill, verticalAlignment=VAlign_Fill}}
  }
  Image_Disabled {
   widgetClassPath=&Image
   bIsVariable=true
   parent=Overlay_4
   widget=Image_Disabled
   properties {
    visibility=HitTestInvisible
    brush {
     bIsDynamicallyLoaded=false
     drawAs=Box
     imageType=NoImage
     mirroring=NoMirror
     resourceName=None
     resourceObject=/Game/Athena/HUD/MiniMap/Art/M_UI_Quest_Container_S19_Fill_White
     tiling=NoTile
     imageSize=<64 64>
     margin=<ltrb 0.5 0.5 0.5 0.5>
     outlineSettings {
      bUseBrushTransparency=false
      roundingType=HalfHeightRadius
      width=0
      color {colorUseRule=UseColor_Specified, specifiedColor=<rgba 0 0 0 0>}
      cornerRadii=<0 0 0 0>
     }
     tintColor {colorUseRule=UseColor_Specified, specifiedColor=<rgba 0 0.0056 0.1329 0.33>}
     uVRegion {bIsValid=false, max=<0 0>, min=<0 0>}
    }
   }
   slot {class=OverlaySlot, properties {horizontalAlignment=HAlign_Fill, verticalAlignment=VAlign_Fill}}
  }
 ]
}
